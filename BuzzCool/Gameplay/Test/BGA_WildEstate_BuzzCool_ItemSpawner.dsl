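; Test interactable actor that spawns a single item pickup, labeled as a Homer Clone Donut, when a pawn interacts with it.
; On interact, it sets HasInteracted and spawns one Spawned Item pickup at a fixed offset from its location.
; HasInteracted blocks further interaction and switches the sub-prompt from empty to a round-active hounds message, which does not match the spawner's behavior.
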(event EventBeginPlay)

(event Collision|EventActorBeginOverlap
  (params (OtherActor :type Actor%20Object%20Reference)))

(event EventTick
  (params (DeltaSeconds :type Float%20%28single-precision%29)))

(event Interaction|EventBlueprintOnInteract
  (params (InteractingPawn :type Fort%20Pawn%20Object%20Reference) (InteractionBeingAttempted :type EInteractionBeingAttempted%20Enum :default "FirstInteraction"))
  (Variables|Default|SetHasInteracted true)
  (Inventory|SpawnItemVariantPickupinWorld (Utilities|Struct|MakeSpawnItemVariantParams (Variables|Default|GetSpawned%20Item) "(GameplayTags=)" 1 (+ (Transformation|GetActorLocation) "0.000000,100.000000,30.000000") "0.000000,0.000000,1.000000" 0 true true true)))

(fn BlueprintCanInteract
  (params (InteractingPawn :type Fort%20Pawn%20Object%20Reference) (InteractionBeingAttempted :type EInteractionBeingAttempted%20Enum :default "FirstInteraction") (InteractionType :type TInteractionType%20Enum :default "IT_NoInteraction"))
  (return (not (Variables|Default|GetHasInteracted))))

(fn BlueprintGetInteractionString
  (params (InteractingPawn :type Fort%20Pawn%20Object%20Reference) (InteractionBeingAttempted :type EInteractionBeingAttempted%20Enum :default "FirstInteraction"))
  (return "NSLOCTEXT(\"[F2E836EBCF26DCB7FE95150B8BDE804D]\", \"021E8C1C4B5B7518F34BCE8B8245F975\", \"Homer Clone Donut\\r\\nFYI: Only One Will Spawn\")"))

(fn BlueprintGetSubInteractionString
  (params (InteractingPawn :type Fort%20Pawn%20Object%20Reference) (InteractionBeingAttempted :type EInteractionBeingAttempted%20Enum :default "FirstInteraction"))
  (if (Variables|Default|GetHasInteracted)
    (return "NSLOCTEXT(\"[F2E836EBCF26DCB7FE95150B8BDE804D]\", \"74028B8945A134BCDB752EB2CD90AA47\", \"Round Active, The Hounds are Currently Released!\")")
    (else
      (return ""))))

(fn ConstructionScript)
